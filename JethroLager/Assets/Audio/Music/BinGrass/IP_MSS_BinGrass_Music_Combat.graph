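# MetaSound source for the BinGrass boss encounter's combat music, with an intro stinger, looping percussion and "else" combat stems, and a death stinger.
# Harmonix metronomes at 172 BPM quantize loop starts, retriggers and the boss-phase-driven ending, which fades the stems under the stinger.
# Distance falloff and a pause fade scale the stereo output.
# The source finishes after the ending stinger or pause fade-out.
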MetaSoundSource /JethroLager/Assets/Audio/Music/BinGrass/IP_MSS_BinGrass_Music_Combat
inputs [
 CombatLoop.Else {type=WaveAsset}
 CombatLoop.Perc {type=WaveAsset}
 EndingStinger {type=WaveAsset}
 Interp Time {type=Time}
 IntroStinger {type=WaveAsset}
 MasterVolume {type=Float}
 Mix.Combat.Else {type=Float}
 Mix.Combat.Master {type=Float}
 Mix.Intro {type=Float}
 Mix.Perc {type=Float}
 MusicEncounter.Attenuation.Volume.FalloffDistance {type=Float}
 MusicEncounter.Attenuation.Volume.InnerRadius {type=Float}
 MusicEncounter.bIsPaused {type=Bool}
 MusicEncounter.Boss.Phase {type=Int32}
 MusicEncounter.PauseFadeOutTime {type=Time}
 UE.Attenuation.Distance {type=Float}
 UE.Source.OnPlay {type=Trigger}
]
interfaces [UE.Attenuation:1.0, UE.OutputFormat.Stereo:1.0, UE.Source.OneShot:1.0, UE.Source:1.1]
outputs [
 "UE.OutputFormat.Stereo.Audio:0" {type=Audio}
 "UE.OutputFormat.Stereo.Audio:1" {type=Audio}
 UE.Source.OneShot.OnFinished {type=Trigger}
]
pages [Default {
  input_defaults {
   CombatLoop.Else=/JethroLager/Assets/Audio/Music/BinGrass/SoundWaves/IP_MX_S35_BinGrass_Phase_1_Loop_Else
   CombatLoop.Perc=/JethroLager/Assets/Audio/Music/BinGrass/SoundWaves/IP_MX_S35_BinGrass_Phase_1_Loop_Perc
   EndingStinger=/JethroLager/Assets/Audio/Music/BinGrass/SoundWaves/IP_MX_S35_BinGrass_Death_STNG
   Interp Time=0.68
   IntroStinger=/JethroLager/Assets/Audio/Music/BinGrass/SoundWaves/IP_MX_S35_BinGrass_Phase_0_to_1_STNG
   MasterVolume=1.407
   Mix.Combat.Else=1
   Mix.Combat.Master=1
   Mix.Intro=1
   Mix.Perc=1
   MusicEncounter.Attenuation.Volume.FalloffDistance=14000
   MusicEncounter.Attenuation.Volume.InnerRadius=6000
   MusicEncounter.Boss.Phase=0
   MusicEncounter.PauseFadeOutTime=5
   MusicEncounter.bIsPaused=null
   UE.Attenuation.Distance=0
   UE.Source.OnPlay=null
  }
  nodes [
   n1 = VariableDeferredAccessor.Trigger(Variable=InitVariable.Trigger().Variable)
   n2 = HarmonixNodes.MidiClockSubdivisionTrigger(MIDI Clock=HarmonixNodes.Metronome(Transport=HarmonixNodes.TriggerToTransport(Play=^UE.Source.OnPlay, Stop=@n1.Value).Transport) {also {defaults {Tempo=172}}}.MIDI Clock) {also {defaults {Grid Size Multiplier=2, Offset Multiplier=2}}}
   n3 = HarmonixNodes.Metronome(Transport=HarmonixNodes.TriggerToTransport(Play=&MSP_ScheduleTrigger(Schedule=^UE.Source.OnPlay, In Trigger=@n2.Trigger Out).Scheduled Trigger).Transport) {also {defaults {Tempo=172}}}
   n4 = VariableAccessor.Trigger(Variable=VariableMutator.Trigger(
     Value=HarmonixNodes.MidiClockSubdivisionTrigger(MIDI Clock=@n3.MIDI Clock) {also {defaults {Base Grid Size=0, Grid Size Multiplier=16}}}.Trigger Out
     Variable=@n1.Variable
    ).Variable)
   n5 = HarmonixNodes.MidiClockSubdivisionTrigger(MIDI Clock=@n3.MIDI Clock)
   n6 = InitVariable.Trigger()
   n7 = InitVariable.Trigger()
   TriggerOnValueChange_1 = UE.Trigger On Value Change:Int32(Value=^MusicEncounter.Boss.Phase)
   n8 = /FortMetasound/Patches/MSP_SwitchOnInt_4(Int=^MusicEncounter.Boss.Phase, Compare=@TriggerOnValueChange_1.Trigger)
   n9 = &MSP_ScheduleTrigger(
    In Trigger=VariableAccessor.Trigger(Variable=VariableMutator.Trigger(Value=@n5.Trigger Out, Variable=@n6.Variable).Variable).Value
    Schedule=VariableAccessor.Trigger(Variable=VariableMutator.Trigger(Variable=@n7.Variable, Value=@n8.Out 1).Variable).Value
   )
   n10 = Value.Value:Float(Set=@n9.Scheduled Trigger) {also {defaults {Init Value=1}}}
   TriggerDelay_1 = UE.Trigger Delay(Delay Time=^Interp Time, In=@n10.On Set)
   n11 = VariableAccessor.Trigger(Variable=VariableMutator.Trigger(Variable=InitVariable.Trigger().Variable, Value=@TriggerDelay_1.Out).Variable)
   WavePlayer_0 = UE.Wave Player:Stereo(
    Wave Asset=^CombatLoop.Perc
    Play=VariableAccessor.Trigger(Variable=@n4.Variable).Value
    Stop=VariableAccessor.Trigger(Variable=@n11.Variable).Value
   ) {also {version [1,1]}}
   WavePlayer_2 = UE.Wave Player:Stereo(Wave Asset=^CombatLoop.Else, Play=@n4.Value, Stop=@n11.Value) {also {version [1,1]}}
   AudioMixerAudioMixerStereo2_0 = "AudioMixer.Audio Mixer (Stereo, 2)"(
    Gain 0=^Mix.Combat.Else
    Gain 1=^Mix.Perc
    In 1 L=@WavePlayer_0.Out Left
    In 1 R=@WavePlayer_0.Out Right
    In 0 L=@WavePlayer_2.Out Left
    In 0 R=@WavePlayer_2.Out Right
   )
   n12 = VariableMutator.Float(Variable=InitVariable.Float().Variable, Value=UE.InterpTo:Audio(Interp Time=^Interp Time, Target=@n10.Output Value).Value)
   WavePlayer_1 = UE.Wave Player:Stereo(Wave Asset=^EndingStinger, Play=@n9.Scheduled Trigger) {also {version [1,1]}}
   WavePlayer_3 = UE.Wave Player:Stereo(Wave Asset=^IntroStinger, Play=^UE.Source.OnPlay) {also {version [1,1]}}
   AudioMixerAudioMixerStereo3_0 = "AudioMixer.Audio Mixer (Stereo, 3)"(
    Gain 0=^Mix.Intro
    In 1 L=@AudioMixerAudioMixerStereo2_0.Out L
    In 1 R=@AudioMixerAudioMixerStereo2_0.Out R
    Gain 1=UE.Multiply:Float(PrimaryOperand=^Mix.Combat.Master, AdditionalOperands=VariableAccessor.Float(Variable=@n12.Variable).Value).Out
    In 2 L=@WavePlayer_1.Out Left
    In 2 R=@WavePlayer_1.Out Right
    In 0 L=@WavePlayer_3.Out Left
    In 0 R=@WavePlayer_3.Out Right
   )
   TriggerOnValueChange_0 = UE.Trigger On Value Change:Bool(Value=^MusicEncounter.bIsPaused)
   n13 = TriggerCompare.Clamp:Bool(A=^MusicEncounter.bIsPaused, Compare=@TriggerOnValueChange_0.Trigger) {also {defaults {B="<MetaSoundBuilderResult.SUCCEEDED: 0>"}}}
   n14 = Value.Value:Float(Reset=@n13.False, Set=@n13.True) {also {defaults {Init Value=1, Target Value=0}}}
   n15 = UE.Add:Float(AdditionalOperands=^MusicEncounter.Attenuation.Volume.FalloffDistance, PrimaryOperand=^MusicEncounter.Attenuation.Volume.InnerRadius)
   n16 = UE.InterpTo:Audio(Target=^UE.Attenuation.Distance) {also {defaults {Interp Time=0.5}}}
   n17 = UE.Multiply:Float(
    PrimaryOperand=^MasterVolume
    AdditionalOperands=UE.Multiply:Float(
     PrimaryOperand=UE.InterpTo:Audio(Interp Time=^MusicEncounter.PauseFadeOutTime, Target=@n14.Output Value).Value
     AdditionalOperands=MapRange.MapRange:Float(In Range A=^MusicEncounter.Attenuation.Volume.InnerRadius, In Range B=@n15.Out, In=@n16.Value) {also {defaults {Out Range A=1, Out Range B=0}}}.Out Value
    ).Out
   )
   Multiply_0 = UE.Multiply:Audio by Float(PrimaryOperand=@AudioMixerAudioMixerStereo3_0.Out R, AdditionalOperands=@n17.Out)
   Multiply_1 = UE.Multiply:Audio by Float(PrimaryOperand=@AudioMixerAudioMixerStereo3_0.Out L, AdditionalOperands=@n17.Out)
   TriggerDelay_0 = UE.Trigger Delay(Delay Time=^MusicEncounter.PauseFadeOutTime, Reset=@n14.On Reset, In=@n14.On Set)
   TriggerAnyTriggerAny2_0 = "TriggerAny.Trigger Any (2)"(In 0=@TriggerDelay_0.Out, In 1=@WavePlayer_1.On Finished)
  ]
  links [
   {dst="@UE.OutputFormat.Stereo.Audio:1", src=Multiply_0.Out}
   {dst="@UE.OutputFormat.Stereo.Audio:0", src=Multiply_1.Out}
   {dst="@UE.Source.OneShot.OnFinished", src=TriggerAnyTriggerAny2_0.Out}
  ]
 }]
properties {
 assetImportData=&self:AssetImportData
 numChannels=2
 outputFormat=Stereo
 sampleRate=48000
 soundClassObject=&Fort_Music_NPC
}
